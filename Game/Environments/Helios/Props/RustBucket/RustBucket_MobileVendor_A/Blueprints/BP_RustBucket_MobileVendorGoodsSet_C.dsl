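; Construction script for the RustBucket mobile vendor goods set C prop, which gives it a random color variant.
; It calls the parent script, resets the static mesh component to unit scale, then writes a random entry from the color swatch array into building texture data slot 0.
; It reads that data back and applies the first entry's tint through SetColoronSecondaryMesh to both the static mesh component and the vendor cart mesh.
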(fn ConstructionScript
  (|CallParentFunction)
  (Transformation|SetRelativeScale3D (Variables|Debug|GetStaticMeshComponent) "1.000000,1.000000,1.000000")
  (if (> (Utilities|Array|Length (Variables|Default|GetAR_ColorSwatches)) 0)
    (CallFunction|SetBuildingTextureData 0 (Utilities|Array|RandomArrayItem (Variables|Default|GetAR_ColorSwatches)))
    (bind _outtexturedataarray (Building|GetBuildingTextureData))
    (if (> (Utilities|Array|Length _outtexturedataarray) 0)
      (Default|SetColoronSecondaryMesh (Math|Conversions|ToLinearColor(Color) (Variables|/Script/FortniteGame.BuildingTextureData|GetTintColor (Utilities|Array|Get(acopy) _outtexturedataarray))) (Variables|Debug|GetStaticMeshComponent))
      (Default|SetColoronSecondaryMesh (Math|Conversions|ToLinearColor(Color) (Variables|/Script/FortniteGame.BuildingTextureData|GetTintColor (Utilities|Array|Get(acopy) _outtexturedataarray))) (Variables|Default|GetSM_RustBucket_MobileVendorCart_A)))))
